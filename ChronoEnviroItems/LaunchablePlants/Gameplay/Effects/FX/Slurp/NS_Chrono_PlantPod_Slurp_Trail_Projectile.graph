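# Niagara system for a slurp trail effect on the launchable plant pod projectile in the Chrono environment items, judging by its name and path.
# The file lists no emitters, user variables, data interfaces or dynamic inputs.
# Its only listed module is the stock SystemState, so the file shows no particle spawning, simulation or rendering behavior.
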NiagaraSystem /ChronoEnviroItems/LaunchablePlants/Gameplay/Effects/FX/Slurp/NS_Chrono_PlantPod_Slurp_Trail_Projectile
dependencies {usedDataInterfaces [], usedDynamicInputs [], usedModules [&SystemState]}
summary {systemName=NS_Chrono_PlantPod_Slurp_Trail_Projectile, emitters [], userVariables []}
